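# Editor utility list-entry widget that shows one actor's display name as a row in a list view.
# When given its O_GenericActorData item, it blocking-loads the item's soft actor reference and writes the actor's display name into a text block.
# ValidateVisuals runs on construct and selection change, tinting the border and text according to whether the item is selected.
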WidgetBlueprint /Game/NeverCook/EditorUtilities/Generic/GenericActorDisplay/EUW_GenericActorItem
graphs [
 EventGraph {nodes [
   Event Construct()
   > ValidateVisuals()
   n1 = Event OnListItemObjectSet()
   > n2 = Cast To O_GenericActorData!(execute, Object=@n1.ListItemObject) -> (then, CastFailed, AsO Generic Actor Data:O Generic Actor Data Object Reference) {is=title K2Node_DynamicCast}
   > n3 = LoadAsset_Blocking(Asset=Get Actor!(self:O Generic Actor Data Object Reference=@n2) -> (Actor:Actor Soft Object Reference))
   > n4 = Cast To Actor!(execute, Object=@n3) -> (then, CastFailed, AsActor:Actor Object Reference) {is=title K2Node_DynamicCast}
   > "SetText (Text)"(self=Get TextBlock_31() -> (TextBlock_31:Text Object Reference), InText="To Text (String)"(InString=GetDisplayName(Object=@n4)))
   Event On Item Selection Changed!() -> (OutputDelegate, then, bIsSelected:bool) {is=title K2Node_Event}
   > ValidateVisuals()
  ]}
 ValidateVisuals {nodes [
   ValidateVisuals() {is=title K2Node_FunctionEntry}
   > SetBrushColor(
    self=Get Border_0() -> (Border_0:Border Object Reference)
    InBrushColor=SelectColor(
     A=(R=0.0433,G=0.0902,B=0.1927,A=1)
     B=(R=0.0312,G=0.0312,B=0.0312,A=1)
     bPickA=IsListItemSelected!(UserListEntry:User List Entry Interface) -> (ReturnValue:bool)
    )
   )
   n1 = IsListItemSelected()
   > SetColorAndOpacity(
    self=Get TextBlock_31()
    InColorAndOpacity=Make Slate Color(SpecifiedColor=SelectColor(A=(R=1,G=1,B=1,A=1), B=(R=0.7812,G=0.7812,B=0.7812,A=1), bPickA=@n1))
   )
  ]}
]
variables []
widgets {
 info {inheritedWidgetCount=0, parentClass=&UserWidget, rootWidgetClass=&StackBox, widgetCount=3}
 widgets [
  StackBox_0 {widgetClassPath=&StackBox, widget=StackBox_0, properties {orientation=Orient_Vertical}}
  Border_0 {
   widgetClassPath=&Border
   bIsVariable=true
   parent=StackBox_0
   widget=Border_0
   properties {brushColor=<rgba 0 0 0 0.55>}
   slot {class=StackBoxSlot}
  }
  TextBlock_31 {
   widgetClassPath=&TextBlock
   bIsVariable=true
   parent=Border_0
   widget=TextBlock_31
   properties {
    text=Actor Display
    font {
     bForceMonospaced=false
     bMaterialIsStencil=false
     fontMaterial=null
     fontObject=&Roboto
     letterSpacing=0
     monospacedWidth=1
     size=9
     skewAmount=0
     typefaceFontName=Bold
     outlineSettings {
      bApplyOutlineToDropShadows=false
      bMiteredCorners=false
      bSeparateFillAlpha=false
      outlineMaterial=null
      outlineSize=0
      outlineColor=<rgba 0 0 0 1>
     }
    }
   }
   slot {class=BorderSlot}
  }
 ]
}
