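# Editor utility widget that shows a titled list of level actors and, on clicking one, focuses the viewport on it and selects it.
# SetDisplayData wraps each actor in an O_GenericActorData holding its soft reference, bounds and, when available, actor descriptor.
# TrySelectAndFocusData blocking-loads the soft reference for current bounds and selection.
# If that reference is invalid, it focuses on the cached bounds instead.
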EditorUtilityWidgetBlueprint /Game/NeverCook/EditorUtilities/Generic/GenericActorDisplay/EUW_GenericActorDisplay
graphs [
 EventGraph {nodes [
   n1 = "On Item Clicked (EditorUtilityListView_107)"!() -> (OutputDelegate, then, Item:Object Reference) {is=title K2Node_ComponentBoundEvent}
   > n2 = Cast To O_GenericActorData!(execute, Object=@n1.Item) -> (then, CastFailed, AsO Generic Actor Data:O Generic Actor Data Object Reference) {is=title K2Node_DynamicCast}
   > TrySelectAndFocusData!(execute, self, inData:O Generic Actor Data Object Reference=@n2) -> (then)
   Event PreConstruct()
   Event Construct()
   Event Tick()
  ]}
 SetDisplayData {nodes [
   n1 = SetDisplayData!() -> (then, Title:String, Actors:Array of Actor Object References) {is=title K2Node_FunctionEntry}
   > ClearListItems(self=Get EditorUtilityListView_107() -> (EditorUtilityListView_107:Editor Utility List View Object Reference))
   > "SetText (Text)"(self=Get TextBlock_31() -> (TextBlock_31:Text Object Reference), InText="To Text (String)"(InString=@n1))
   > n2 = For Each Loop(Array:Array of Actor Object References=@n1.Actors) -> (Array Element:Actor Object Reference)
   > n3 = Construct Object from Class(Class=&/O_GenericActorData.O_GenericActorData_C) -> (ReturnValue:O Generic Actor Data Object Reference)
   > n4 = GetActorDescsForActors!(execute, InActors:Array of Actor Object References=Make Array("[0]":Actor Object Reference=@n2) -> (Array:Array of Actor Object References)) -> (then, OutActorDescs:Array of Actor Desc Structures, ReturnValue:bool)
   > n5 = Branch(Condition=@n4.ReturnValue)
   > n6 = Set ActorDesc!(
    execute
    ActorDesc:Actor Desc Structure="Get (a copy)"(Array:Array of Actor Desc Structures=@n4) -> (Output:Actor Desc Structure)
    self:O Generic Actor Data Object Reference=@n3
   ) -> (then, Output_Get:Actor Desc Structure)
   Set Actor!(
    execute=[@n5.else, @n6.then]
    Actor:Actor Soft Object Reference=Make Soft Reference!(Input:Actor Object Reference=@n2) -> (Output:Actor Soft Object Reference)
    self:O Generic Actor Data Object Reference=@n3
   ) -> (then, Output_Get:Actor Soft Object Reference)
   n7 = Get Actor Bounds(self=@n2, bOnlyCollidingComponents=true)
   > Set Bounds!(
    execute
    Bounds:Box Sphere Bounds Structure=MakeBoxSphereBounds!(Origin:Vector=@n7, BoxExtent:Vector=@n7.BoxExtent, SphereRadius:float=Vector Length(A=@n7.BoxExtent)) -> (ReturnValue:Box Sphere Bounds Structure)
    self:O Generic Actor Data Object Reference=@n3
   ) -> (then, Output_Get:Box Sphere Bounds Structure)
   > AddItem!(execute, self:List View Object Reference=Get EditorUtilityListView_107(), Item:Object Reference=@n3) -> (then)
  ]}
 TrySelectAndFocusData {nodes [
   n1 = TrySelectAndFocusData!() -> (then, inData:O Generic Actor Data Object Reference) {is=title K2Node_FunctionEntry}
   > Set l_data(l_data:O Generic Actor Data Object Reference=@n1) -> (Output_Get:O Generic Actor Data Object Reference)
   n2 = Get Actor!(self:O Generic Actor Data Object Reference=Get l_data() -> (l_data:O Generic Actor Data Object Reference)) -> (Actor:Actor Soft Object Reference)
   > n3 = Branch(Condition=IsValidSoftObjectReference(SoftObjectReference=@n2))
   > n4 = LoadAsset_Blocking(Asset=@n2)
   > n5 = Cast To Actor!(execute, Object=@n4) -> (then, CastFailed, AsActor:Actor Object Reference) {is=title K2Node_DynamicCast}
   > n6 = Set l_actor(l_actor:Actor Object Reference=@n5) -> (Output_Get:Actor Object Reference)
   > n7 = Get Actor Bounds!(execute:Exec, self=@n6, bOnlyCollidingComponents=true, bIncludeFromChildActors) -> (then:Exec, Origin, BoxExtent)
   > Set l_center(l_center:Vector=@n7) -> (Output_Get:Vector)
   > n8 = Set l_extent(l_extent:Vector=@n7.BoxExtent) -> (Output_Get:Vector)
   Focus Bounds in Viewport!(" ":Exec=[@n8.then, @n11.then], center:Vector=Get l_center() -> (l_center:Vector), bounds:Vector=Get l_extent() -> (l_extent:Vector)) -> ("  ":Exec) {is=title K2Node_MacroInstance}
   n9 = Get l_actor() -> (l_actor:Actor Object Reference)
   > Is Valid(InputObject=@n9)
   > SetSelectedLevelActors(self=Get EditorActorSubsystem(), ActorsToSelect=Make Array("[0]"=@n9))
   n10 = BreakBoxSphereBounds!(InBoxSphereBounds:"Box Sphere Bounds Structure (by ref)"=Get Bounds!(self:O Generic Actor Data Object Reference=Get l_data()) -> (Bounds:Box Sphere Bounds Structure)) -> (Origin:Vector, BoxExtent:Vector, SphereRadius:float)
   Set l_center(execute=@n3.else, l_center=@n10)
   > n11 = Set l_extent(l_extent=@n10.BoxExtent)
  ]}
]
variables []
widgets {
 info {rootWidgetClass=&StackBox, widgetCount=5}
 widgets [
  StackBox_0 {widgetClassPath=&StackBox, widget=StackBox_0, properties {orientation=Orient_Vertical}}
  Border_0 {widgetClassPath=&Border, parent=StackBox_0, widget=Border_0, properties {brushColor=<rgba 0 0 0 0.55>}, slot {class=StackBoxSlot}}
  TextBlock_31 {
   widgetClassPath=&TextBlock
   bIsVariable=true
   parent=Border_0
   widget=TextBlock_31
   properties {
    text=Actor Display
    font {
     bForceMonospaced=false
     bMaterialIsStencil=false
     fontMaterial=null
     fontObject=&Roboto
     letterSpacing=0
     monospacedWidth=1
     size=12
     skewAmount=0
     typefaceFontName=Bold
     outlineSettings {
      bApplyOutlineToDropShadows=false
      bMiteredCorners=false
      bSeparateFillAlpha=false
      outlineMaterial=null
      outlineSize=0
      outlineColor=<rgba 0 0 0 1>
     }
    }
   }
   slot {class=BorderSlot}
  }
  Border_1 {
   widgetClassPath=&Border
   parent=StackBox_0
   widget=Border_1
   properties {brushColor=<rgba 0 0 0 0.421>}
   slot {class=StackBoxSlot, properties {padding=<ltrb 10 10 10 10>, size {sizeRule=Fill, value=1}}}
  }
  EditorUtilityListView_107 {
   widgetClassPath=/Script/Blutility.EditorUtilityListView
   bIsVariable=true
   parent=Border_1
   widget=EditorUtilityListView_107
   properties {
    entryWidgetClass=&/EUW_GenericActorItem.EUW_GenericActorItem_C
    verticalEntrySpacing=3
   }
   slot {class=BorderSlot}
  }
 ]
}
